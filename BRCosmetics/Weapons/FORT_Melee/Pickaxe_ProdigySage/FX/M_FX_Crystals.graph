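# Translucent, lit crystal particle material for the Prodigy Sage pickaxe effects, usable on Niagara sprites.
# The crystal texture's green channel lerps a brightness factor that multiplies particle color and feeds the shared emissive adjustments function.
# Particle alpha is multiplied by the texture's red channel as a shape mask.
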Material /BRCosmetics/Weapons/FORT_Melee/Pickaxe_ProdigySage/FX/M_FX_Crystals
graph {
 properties {bUsedWithNiagaraSprites=true, blendMode=BLEND_Translucent, pixelDepthOffsetMode=PDOM_Legacy, usedShadingModels=Default Lit}
 nodes [
  n1 = ParticleColor()
  n2 = TextureSample() {texture=&/T_FX_Crystal}
  MaterialFunctionCall(Emissive=Multiply(A=@n1.RGB, B=LinearInterpolate(Alpha=@n2.G) {constA=0.2, constB=1.2})) {materialFunction=&MF_FX_EmissiveAdjustments}
  Multiply(A=@n1.A, B=@n2.R)
 ]
}
