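# Additive unlit glow material for Niagara mesh particles in the Prodigy Sage pickaxe's star effect.
# Particle color RGB runs through the shared emissive adjustments function.
# An inverted V texture coordinate is raised to a power and multiplied by particle alpha, forming a gradient falloff along the mesh UVs.
# A view-angle fade function and an AngleFade scalar are also present, with no connections shown.
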Material /BRCosmetics/Weapons/FORT_Melee/Pickaxe_ProdigySage/FX/M_FX_Glow_Star
graph {
 properties {
  bUsedWithNiagaraMeshParticles=true
  blendMode=BLEND_Additive
  pixelDepthOffsetMode=PDOM_Legacy
  shadingModel=MSM_Unlit
  usedShadingModels=Unlit
 }
 nodes [
  n1 = ParticleColor()
  MaterialFunctionCall(Emissive=@n1.RGB) {materialFunction=&MF_FX_EmissiveAdjustments}
  MaterialFunctionCall() {materialFunction=&FadeBasedOnViewAngle}
  n2 = MaterialFunctionCall(Float2=TextureCoordinate()) {materialFunction=&BreakOutFloat2Components}
  Multiply(A=Power(Base=OneMinus(None=@n2.G)), B=@n1.A)
  ScalarParameter() {parameterName=AngleFade}
 ]
}
