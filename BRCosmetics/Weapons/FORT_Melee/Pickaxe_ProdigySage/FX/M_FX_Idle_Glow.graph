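# Translucent two-sided particle material for Niagara sprites, named as an idle glow effect for the Prodigy Sage pickaxe.
# Particle color RGB runs through a shared emissive-adjustment function to set the glow color.
# An alpha mask multiplies particle alpha, the green channel of a crystal texture, and a view-angle fade with a tunable FadeAngle parameter.
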Material /BRCosmetics/Weapons/FORT_Melee/Pickaxe_ProdigySage/FX/M_FX_Idle_Glow
graph {
 properties {bUsedWithNiagaraSprites=true, blendMode=BLEND_Translucent, pixelDepthOffsetMode=PDOM_Legacy, twoSided=true, usedShadingModels=Default Lit}
 nodes [
  n1 = ParticleColor()
  MaterialFunctionCall(Emissive=@n1.RGB) {materialFunction=&MF_FX_EmissiveAdjustments}
  Multiply(
   A=Multiply(A=@n1.A, B=TextureSample() {texture=&/T_FX_Crystal}.G)
   B=MaterialFunctionCall("FadeSensitivity (optional)"=ScalarParameter() {defaultValue=3, parameterName=FadeAngle}) {materialFunction=&FadeBasedOnViewAngle}.Result
  )
 ]
}
