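# Idle Niagara effect for the Prodigy Sage pickaxe, adding looping glows to the weapon's gem spikes and center.
# Glow pins eight sprites to skeletal mesh sockets, folded into four size groups, with two renderers and perpendicular socket-aligned facings.
# Center_Glow places two rotating sprites at socket 8, while Center_Glow001 adds a mesh glow.
# All three derive opacity from User.Gems_EffectAmount through scratch dynamic inputs, scaled by time of day.
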NiagaraSystem /BRCosmetics/Weapons/FORT_Melee/Pickaxe_ProdigySage/FX/NS_Pickaxe_ProdigySage_Idle
dependencies {
 usedDataInterfaces [&NiagaraDataInterfaceCurve, &NiagaraDataInterfaceArrayFloat, &NiagaraDataInterfaceArrayFloat3]
 usedDynamicInputs []
 usedModules [
  &self:SystemSpawnScript.SystemScriptSource.SystemScriptGraph.NiagaraNodeAssignment_1.SetVariables_#1
  &SystemState
 ]
}
emitters {
 Center_Glow {
  emitterUpdateScript [
   &EmitterState(
    Inactive Response=::NewEnumerator1
    Life Cycle Mode=::NewEnumerator1
    Loop Behavior=::NewEnumerator1
    Loop Duration Mode=::NewEnumerator1
    UseLoopDelay=false
   )
   &SpawnBurst_Instantaneous(Spawn Count=2, Use Spawn Probability=false, UseLoopCountLimit=false)
  ]
  particleSpawnScript [
   &InitializeParticle(
    Color=<rgba 0 0.5833 1 0.2>
    Lifetime=2
    Position Offset=&SelectVectorFromArray(
     Direct Array Index=&ReturnExecIndex()
     Vector Selection Array=DataInterface() {propertyValues {FloatData [<0 1 0>, <0 0 0>]}}
    )
    Sprite Size Mode=::NewEnumerator3
    Uniform Sprite Size=120
    UsePositionOffset=true
    Write Material Random=true
   )
   &SkeletalMeshLocation(
    Bone / Socket Index=8
    Interpolate Mesh Sampling=false
    MaskSpawnGroup=false
    Mesh Sampling Type=::NewEnumerator5
    Sampled Position Offset=&SelectVectorFromArray(
     Direct Array Index=&ReturnExecIndex()
     Vector Selection Array=DataInterface() {propertyValues {FloatData [<1 0 0>, <-1 0 0>]}}
    )
    Skeletal Mesh=$System.Skeletal Mesh
    Socket Sampling Mode=::NewEnumerator5
   )
  ]
  particleUpdateScript [
   &ParticleState(Kill Particles When Lifetime Has Elapsed=false, Loop Particles Lifetime=true)
   &SpriteRotationRate(Rotation Rate=4)
   &CameraOffset(Camera Offset Amount=10)
   /SetVariables(Particles.StarOpacity:NiagaraFloat=&self:ScratchDynamicInput_01(GemsAmount:NiagaraFloat=$User.Gems_EffectAmount)) {isSetParameters=true, module=SetVariables_#2}
   &ScaleColor(
    Scale Alpha=&FloatFromCurve(
     FloatCurve=DataInterface() {propertyValues {Curve {keys [<at 0 1 RCIM_Cubic>,<at 0.5 0.6 RCIM_Cubic>,<at 1 1 RCIM_Cubic>]}}}
     Scale Curve=&Multiply_Float(
      A=&FNTimeOfDayScalar(Night Scale=0.4, Morning Scale=0.6, Day Scale=1.1, Evening Scale=0.6)
      B=$Particles.StarOpacity
     )
    )
    ScaleA=true
    ScaleRGB=true
   )
  ]
  rendererClasses [&NiagaraSpriteRendererProperties]
  renderers [{rendererClass=&NiagaraSpriteRendererProperties}]
 }
 Center_Glow001 {
  emitterUpdateScript [
   &EmitterState(
    Inactive Response=::NewEnumerator1
    Life Cycle Mode=::NewEnumerator1
    Loop Behavior=::NewEnumerator1
    Loop Duration Mode=::NewEnumerator1
    UseLoopDelay=false
   )
   &SpawnBurst_Instantaneous(Use Spawn Probability=false, UseLoopCountLimit=false)
   &SpawnBurst_Instantaneous(Spawn Count=2, Spawn Time=0.6977, Use Spawn Probability=false, UseLoopCountLimit=false) {enabled=false, module=SpawnBurst_Instantaneous001}
  ]
  particleSpawnScript [&InitializeParticle(
    Color=<rgba 0.125 0.875 1.25 0.1>
    Lifetime=2
    Sprite Size Mode=::NewEnumerator3
    Uniform Sprite Size=120
    UsePositionOffset=true
    Write Material Random=true
   )]
  particleUpdateScript [
   &ParticleState(Kill Particles When Lifetime Has Elapsed=false, Loop Particles Lifetime=true)
   /SetVariables(Particles.StarOpacity:NiagaraFloat=&self:ScratchDynamicInput_01(GemsAmount:NiagaraFloat=$User.Gems_EffectAmount)) {isSetParameters=true, module=SetVariables_#2}
   &ScaleColor(
    Scale Alpha=&Multiply_Float(
     A=$Particles.StarOpacity
     B=&FNTimeOfDayScalar(Night Scale=0.7, Morning Scale=1, Day Scale=1.1, Evening Scale=1)
    )
    ScaleA=true
    ScaleRGB=true
   )
  ]
  rendererClasses [&NiagaraMeshRendererProperties]
  renderers [{rendererClass=&NiagaraMeshRendererProperties, rendererIndex=0}]
 }
 Glow {
  emitterUpdateScript [
   &EmitterState(Life Cycle Mode=::NewEnumerator1, Loop Behavior=::NewEnumerator1, Loop Duration Mode=::NewEnumerator1, UseLoopDelay=false)
   &SpawnBurst_Instantaneous(Spawn Count=8, Use Spawn Probability=false, UseLoopCountLimit=false)
  ]
  particleSpawnScript [
   &FN_SetSpawnGroup(Number of Spawn Groups=8)
   &InitializeParticle(
    Color=<rgba 0.1321 1.022 1.2 0.5>
    Sprite Size=<15 22>
    Sprite Size Mode=::NewEnumerator4
    UsePositionOffset=false
    Write Material Random=true
   )
   &InitialMeshOrientation(Mesh Orientation Mode=::NewEnumerator1, Rotation=<1 1 0.5>, Use Rotation Vector=true)
   &CameraOffset(Camera Offset Amount=-5)
   /SetVariables(Particles.SideSize:Vector2f=<13 24>) {isSetParameters=true}
  ]
  particleUpdateScript [
   &ParticleState(
    Kill Particles When Lifetime Has Elapsed=false
    Let Infinitely Lived Particles Die When Emitter Deactivates=true
    Loop Particles Lifetime=true
   )
   &SkeletalMeshLocation(
    Bone / Socket Index=$Particles.SpawnGroup
    MaskSpawnGroup=false
    Mesh Sampling Type=::NewEnumerator5
    Sampled Position Offset=<0 0 6>
    Skeletal Mesh=$System.Skeletal Mesh
    Socket Sampling Mode=::NewEnumerator5
   )
   &ScaleSpriteSize(Scale Sprite Size Mode=::NewEnumerator0, Uniform Scale Factor=1.2)
   /SetVariables(Particles.SpikeGroup:NiagaraInt32=/Niagara/DynamicInputs/Subtract/Subtract_Integer(
     A:NiagaraInt32=$Particles.SpawnGroup
     B:NiagaraInt32=&MakeIntFromFloat(Float=&Lerp_Float(
       A=0
       B=4
       Alpha=&MakeCustomFloatFromBool(
        Bool=&SetBoolByFloatComparison(A=&MakeFloatFromInt(Int=$Particles.SpawnGroup), B=3)
        True Float=1
        False Float=0
       )
      ))
    )) {isSetParameters=true}
   /SetVariables(Particles.SpikeVisibility:NiagaraFloat=&Lerp_Float(
     A=0
     B=1
     Alpha=&MakeCustomFloatFromBool(
      Bool=&SetBoolByIntComparison(A=$Particles.SpikeGroup, B=&MakeIntFromFloat(Float=$User.Test), Comparison Type=::NewEnumerator6)
      True Float=1
      False Float=0
     )
    )) {isSetParameters=true}
   /SetVariables(Particles.Spike_1_Opacity:NiagaraFloat=&self:ScratchDynamicInput(Value:NiagaraFloat=$User.Gems_EffectAmount)) {isSetParameters=true}
   &ScaleColor(
    Scale Alpha=&Multiply_Float(
     A=$Particles.Spike_1_Opacity
     B=&FNTimeOfDayScalar(Night Scale=0.6, Morning Scale=0.7, Evening Scale=0.7)
    )
    Scale RGB=&Multiply_VectorByFloat(Vector=<1 1 1>, Float=1)
    ScaleA=true
    ScaleRGB=false
   ) {module=ScaleColor002}
   &ScaleSpriteSize(
    Initial Sprite Size:Vector2f=$Particles.Initial.SideSize
    Scale Sprite Size Mode=::NewEnumerator0
    Uniform Scale Factor=&SelectFloatFromArray(Direct Array Index=$Particles.SpikeGroup, Float Selection Array=DataInterface() {propertyValues {FloatData [1,1,1.2,1.1]}})
   ) {enabled=false, module=ScaleSpriteSize002}
   &ScaleSpriteSize(
    Scale Sprite Size Mode=::NewEnumerator0
    Uniform Scale Factor=&SelectFloatFromArray(Direct Array Index=$Particles.SpikeGroup, Float Selection Array=DataInterface() {propertyValues {FloatData [1,1,1.2,1.1]}})
   ) {module=ScaleSpriteSize001}
   &SpriteFacingAndAlignment(
    Sprite Alignment=&ReturnMeshOrientationAxis(Mesh Orientation Axis=::ZAxis, Mesh Orientation Quat=$Particles.SkeletalMeshLocation.SampledRotation)
    Sprite Facing=&ReturnMeshOrientationAxis(Mesh Orientation Axis=::YAxis, Mesh Orientation Quat=$Particles.SkeletalMeshLocation.SampledRotation)
    WriteAlignment=true
    WriteFacing=true
   ) {module=SpriteFacingAndAlignment001}
   /SetVariables(Particles.SideAlignment:Vector3f=$StackContext.SpriteAlignment, Particles.SideFacing:Vector3f=$StackContext.SpriteFacing) {isSetParameters=true}
   &SpriteFacingAndAlignment(
    Sprite Alignment=&ReturnMeshOrientationAxis(Mesh Orientation Axis=::ZAxis, Mesh Orientation Quat=$Particles.SkeletalMeshLocation.SampledRotation)
    Sprite Facing=&ReturnMeshOrientationAxis(Mesh Orientation Axis=::XAxis, Mesh Orientation Quat=$Particles.SkeletalMeshLocation.SampledRotation)
    WriteAlignment=true
    WriteFacing=true
   )
  ]
  rendererClasses [&NiagaraSpriteRendererProperties]
  renderers [
   {rendererClass=&NiagaraSpriteRendererProperties}
   {rendererClass=&NiagaraSpriteRendererProperties, rendererIndex=1}
  ]
 }
}
summary {
 systemName=NS_Pickaxe_ProdigySage_Idle
 emitters [
  {emitterName=Glow, rendererClasses [&NiagaraSpriteRendererProperties]}
  {emitterName=Center_Glow, rendererClasses [&NiagaraSpriteRendererProperties]}
  {emitterName=Center_Glow001, rendererClasses [&NiagaraMeshRendererProperties]}
 ]
 userVariables [
  User.bIsFrontEnd {type=NiagaraBool}
  User.bIsFrontEndPreview {type=NiagaraBool}
  User.Gems_EffectAmount {type=NiagaraFloat, defaultValue=0, description=""}
  User.Gems_EffectIntensity {type=NiagaraFloat, defaultValue=1, description=""}
  User.Skeletal Mesh {
   type=NiagaraDataInterfaceSkeletalMesh
   defaultValue {
    struct=NiagaraExt_VariableValue_DataInterface
    value {dataInterface=&self:NiagaraDataInterfaceSkeletalMesh_0}
   }
  }
  User.Test {type=NiagaraFloat, defaultValue=0, description=""}
  User.Transition {type=NiagaraFloat, defaultValue=0, description=""}
  User.WeakPoint int {type=NiagaraInt32}
 ]
}
