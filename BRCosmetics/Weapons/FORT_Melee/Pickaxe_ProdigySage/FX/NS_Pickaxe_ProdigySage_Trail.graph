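# Niagara anim-trail effect for the ProdigySage pickaxe, with four ribbon emitters driven by skeletal mesh bones or sockets.
# Blue and pink underlays fade over life and scale alpha by time of day.
# Four spike trails use separate bone pairs and lifetimes, each blending color and alpha by comparing User.Gems_EffectAmount to its spawn group, weighted by Gems_EffectIntensity.
# A bright center trail spawns conditionally on gem amount compared against five.
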NiagaraSystem /BRCosmetics/Weapons/FORT_Melee/Pickaxe_ProdigySage/FX/NS_Pickaxe_ProdigySage_Trail
dependencies {
 usedDataInterfaces [&NiagaraDataInterfaceCurve, &NiagaraDataInterfaceArrayFloat, &NiagaraDataInterfaceColorCurve]
 usedDynamicInputs []
 usedModules [
  &self:SystemSpawnScript.SystemScriptSource.SystemScriptGraph.NiagaraNodeAssignment_0.SetVariables_#1
  &SystemState
 ]
}
emitters {
 Blue_Underaly {
  emitterUpdateScript [
   &EmitterState()
   &FN_SpawnAnimTrail()
  ]
  particleSpawnScript [
   &FN_InitializeAnimTrailBoneIndecies()
   &FN_InitializeAnimTrail(
    Color=&MultiplyLinearColor(
     Linear Color=<rgba 1 1 1 1>
     Scale Factor=&MakeFloat4FromVectorAndFloat("Vector (RGB)"=&VectorFromFloat(Value=1), "Float (Alpha)"=1.4)
    )
    Lifetime=0.25
    SkeletalMesh=$System.Skeletal Mesh
   )
  ]
  particleUpdateScript [
   &FN_UpdateAnimTrail(Locked Particle RibbonWidth=1, SkeletalMesh=$User.SkeletalMesh) {module=FN_UpdateAnimTrail001}
   &ScaleColor(
    Scale Alpha=&Multiply_Float(
     A=&FloatFromCurve(FloatCurve=DataInterface() {propertyValues {Curve {keys [{tangentMode=RCTM_User, time=0, value=1}, {tangentMode=RCTM_User, time=1, value=0}]}}})
     B=&Multiply_Float(A=0.5, B=&FNTimeOfDayScalar(Night Scale=0.6, Morning Scale=0.7, Evening Scale=0.7))
    )
    Scale RGB=&MakeVectorFromLinearColorRGB(Color=<rgba 0 0.0383 0.1 1>)
    ScaleA=true
    ScaleRGB=true
   ) {module=ScaleColor001}
   &CameraOffset()
  ]
  rendererClasses [&NiagaraRibbonRendererProperties]
  renderers [{rendererClass=&NiagaraRibbonRendererProperties}]
 }
 Center_Trail {
  emitterUpdateScript [
   &EmitterState()
   &FN_SpawnAnimTrail(Number of Anim Trails:NiagaraInt32=&MakeCustomIntFromBool(
     Bool=&SetBoolByFloatComparison(A=$User.Gems_EffectAmount, B=5, Comparison Type=::NewEnumerator4)
     True INT=1
     False INT=0
    ))
  ]
  particleSpawnScript [
   &FN_InitializeAnimTrailBoneIndecies()
   &FN_InitializeAnimTrail(
    AnimTrailWidth=0.35
    Color=&MultiplyLinearColor(
     Linear Color=<rgba 1 1 1 1>
     Scale Factor=&MakeFloat4FromVectorAndFloat("Vector (RGB)"=&VectorFromFloat(Value=2), "Float (Alpha)"=0.1)
    )
    Lifetime=0.11
   )
  ]
  particleUpdateScript [
   &FN_UpdateAnimTrail(Locked Particle RibbonWidth=1, SkeletalMesh=$User.SkeletalMesh) {module=FN_UpdateAnimTrail001}
   &ScaleColor(
    Scale Alpha=&Multiply_Float(
     A=&FloatFromCurve(FloatCurve=DataInterface() {propertyValues {Curve {keys [{tangentMode=RCTM_User, time=0, value=1}, {tangentMode=RCTM_User, time=1, value=0}]}}})
     B=1
    )
    Scale RGB=&MakeVectorFromLinearColorRGB(Color=<rgba 0.48 1.2334 1.6 1>)
    ScaleA=true
    ScaleRGB=true
   ) {module=ScaleColor001}
   &CameraOffset(Camera Offset Amount=5)
  ]
  rendererClasses [&NiagaraRibbonRendererProperties]
  renderers [{rendererClass=&NiagaraRibbonRendererProperties}]
 }
 Pink_Underlay {
  emitterUpdateScript [
   &EmitterState()
   &FN_SpawnAnimTrail()
  ]
  particleSpawnScript [
   &FN_InitializeAnimTrailBoneIndecies()
   &FN_InitializeAnimTrail(
    AnimTrailWidth=0.9
    Color=&MultiplyLinearColor(
     Linear Color=<rgba 1 1 1 1>
     Scale Factor=&MakeFloat4FromVectorAndFloat("Vector (RGB)"=&VectorFromFloat(Value=0.5), "Float (Alpha)"=1)
    )
    Lifetime=0.25
   )
  ]
  particleUpdateScript [
   &FN_UpdateAnimTrail(Locked Particle RibbonWidth=1, SkeletalMesh=$User.SkeletalMesh) {module=FN_UpdateAnimTrail001}
   &ScaleColor(
    Scale Alpha=&Multiply_Float(
     A=&FloatFromCurve(FloatCurve=DataInterface() {propertyValues {Curve {keys [{tangentMode=RCTM_User, time=0, value=1}, {tangentMode=RCTM_User, time=1, value=0}]}}})
     B=&Multiply_Float(A=0.08, B=&FNTimeOfDayScalar(Night Scale=0.6, Morning Scale=0.7, Evening Scale=0.7))
    )
    Scale RGB=&MakeVectorFromLinearColorRGB(Color=<rgba 0.3564 0.0027 0.1796 1>)
    ScaleA=true
    ScaleRGB=true
   ) {module=ScaleColor001}
   &CameraOffset(Camera Offset Amount=5)
  ]
  rendererClasses [&NiagaraRibbonRendererProperties]
  renderers [{rendererClass=&NiagaraRibbonRendererProperties}]
 }
 Spikes_Trail001 {
  emitterUpdateScript [
   &EmitterState()
   &FN_SpawnAnimTrail(Number of Anim Trails=4, Spawn Interval=0)
  ]
  particleSpawnScript [
   &FN_InitializeAnimTrailBoneIndecies(First Bone or Socket Index=2, Second Bone or Socket Index=3)
   &FN_InitializeAnimTrailBoneIndecies(AnimTrail Selection=1, First Bone or Socket Index=4, Second Bone or Socket Index=5) {module=FN_InitializeAnimTrailBoneIndecies001}
   &FN_InitializeAnimTrailBoneIndecies(AnimTrail Selection=2, First Bone or Socket Index=7, Second Bone or Socket Index=6) {module=FN_InitializeAnimTrailBoneIndecies002}
   &FN_InitializeAnimTrailBoneIndecies(AnimTrail Selection=3, First Bone or Socket Index=9, Second Bone or Socket Index=8) {module=FN_InitializeAnimTrailBoneIndecies003}
   &FN_InitializeAnimTrail(
    Lifetime=&Multiply_Float(
     A=&SelectFloatFromArray(Direct Array Index=$Particles.SpawnGroup, Float Selection Array=DataInterface() {propertyValues {FloatData [0.15,0.2,0.25,0.3]}})
     B=1
    )
    SkeletalMesh=$System.Skeletal Mesh
   )
  ]
  particleUpdateScript [
   &FN_UpdateAnimTrail(SkeletalMesh=$User.SkeletalMesh)
   &ScaleColor(
    Scale Alpha=&FloatFromCurve(
     FloatCurve=DataInterface() {propertyValues {Curve {keys [
         {arriveTangent=-2.4911, interpMode=RCIM_Cubic, leaveTangent=-2.4911, tangentMode=RCTM_User, time=0, value=1}
         {interpMode=RCIM_Cubic, tangentMode=RCTM_User, time=1, value=0}
        ]}}}
     Scale Curve=&Lerp_Float(
      A=1.4
      B=0.3
      Alpha=&Multiply_Float(
       A=&MakeCustomFloatFromBool(
        Bool=&SetBoolByFloatComparison(A=$User.Gems_EffectAmount, B=&MakeFloatFromInt(Int=$Particles.SpawnGroup))
        True Float=1
        False Float=0
       )
       B=$User.Gems_EffectIntensity
      )
     )
    )
    Scale RGB=&Lerp_Vector(
     A=&Multiply_Vector(
      A=&MakeVectorFromLinearColorRGB(Color=&ColorFromCurve(ColorCurve=DataInterface() {propertyValues {
          AlphaCurve {keys [<at 0 1 RCIM_Cubic>,<at 1 1 RCIM_Cubic>]}
          BlueCurve {keys [<at 0.004 0.1>,<at 0.301 0.2>]}
          GreenCurve {keys [<at 0.004 0.0507>,<at 0.301 0.1013>]}
          RedCurve {keys [<at 0.004 0.02>,<at 0.301 0.04>]}
         }}))
      B=<1 1 1>
     )
     B=&Multiply_Vector(
      A=&MakeVectorFromLinearColorRGB(Color=&ColorFromCurve(ColorCurve=DataInterface() {propertyValues {
          AlphaCurve {keys [<at 0 1 RCIM_Cubic>,<at 1 1 RCIM_Cubic>]}
          BlueCurve {keys [<at 0 2>,<at 0.1949 0.9>,<at 1 0.6>]}
          GreenCurve {keys [<at 0 0.7399>,<at 0.1949 0.681>,<at 1 0.4>]}
          RedCurve {keys [<at 0 0.0104>,<at 0.1949 0.127>,<at 1 0>]}
         }}))
      B=<1 1 1>
     )
     Alpha=&Multiply_Float(
      A=&MakeCustomFloatFromBool(
       Bool=&SetBoolByFloatComparison(A=$User.Gems_EffectAmount, B=&MakeFloatFromInt(Int=$Particles.SpawnGroup))
       True Float=1
       False Float=0
      )
      B=$User.Gems_EffectIntensity
     )
    )
    ScaleA=true
    ScaleRGB=true
   ) {module=ScaleColor001}
   &ScaleColor(
    Scale Alpha=&FloatFromCurve(
     FloatCurve=DataInterface() {propertyValues {Curve {keys [
         {arriveTangent=-1.6414, interpMode=RCIM_Cubic, leaveTangent=-1.6414, tangentMode=RCTM_User, time=0, value=1}
         {interpMode=RCIM_Cubic, tangentMode=RCTM_User, time=1, value=0}
        ]}}}
     Scale Curve=&Lerp_Float(
      A=0.25
      B=0.2
      Alpha=&Multiply_Float(
       A=&MakeCustomFloatFromBool(
        Bool=&SetBoolByFloatComparison(A=$User.Gems_EffectAmount, B=&MakeFloatFromInt(Int=$Particles.SpawnGroup))
        True Float=1
        False Float=0
       )
       B=$User.Gems_EffectIntensity
      )
     )
    )
    Scale RGB=&Lerp_Vector(
     A=&Multiply_Vector(
      A=&MakeVectorFromLinearColorRGB(Color=&ColorFromCurve(ColorCurve=DataInterface() {propertyValues {
          AlphaCurve {keys [<at 0 1 RCIM_Cubic>,<at 1 1 RCIM_Cubic>]}
          BlueCurve {keys [<at 0.0021 0.5>,<at 0.1949 0.35>,<at 1 0.25>]}
          GreenCurve {keys [<at 0.0021 0.1021>,<at 0.1949 0.1063>,<at 1 0.0958>]}
          RedCurve {keys [<at 0.0021 0.0026>,<at 0.1949 0.0018>,<at 1 0>]}
         }}))
      B=<1 1 1>
     )
     B=&Multiply_Vector(
      A=&MakeVectorFromLinearColorRGB(Color=&ColorFromCurve(ColorCurve=DataInterface() {propertyValues {
          AlphaCurve {keys [<at 0 1 RCIM_Cubic>,<at 1 1 RCIM_Cubic>]}
          BlueCurve {keys [<at 0 2>,<at 0.1949 0.8469>,<at 1 0.333>]}
          GreenCurve {keys [<at 0 0.7399>,<at 0.1949 0.6408>,<at 1 0.2442>]}
          RedCurve {keys [<at 0 0.0104>,<at 0.1949 0.1195>,<at 1 0>]}
         }}))
      B=<1 1 1>
     )
     Alpha=&Multiply_Float(
      A=&MakeCustomFloatFromBool(
       Bool=&SetBoolByFloatComparison(A=$User.Gems_EffectAmount, B=&MakeFloatFromInt(Int=$Particles.SpawnGroup))
       True Float=1
       False Float=0
      )
      B=$User.Gems_EffectIntensity
     )
    )
    ScaleA=true
    ScaleRGB=true
   ) {enabled=false}
   &ParticleState()
   &CameraOffset(Camera Offset Amount=10)
   &ScaleRibbonWidth(Ribbon Width Scale=&FloatFromCurve(
     FloatCurve=DataInterface() {propertyValues {Curve {keys [{tangentMode=RCTM_User, time=0.25, value=1}, {tangentMode=RCTM_User, time=1, value=0.25}]}}}
     Scale Curve=2
    ))
   /SetVariables(
    Particles.SecondaryColor:LinearColor=&MultiplyLinearColor(
     Linear Color=$Particles.Color
     Scale Factor=&MakeVector4(
      X=0.3
      Y=0.3
      Z=0.3
      W=&Multiply_Float(
       A=&MakeFloatFromLinearColor(LinearColor=$Particles.Color, Channel:ENiagaraLinearColor_Channels=::NewEnumerator3)
       B=&Multiply_Float(A=0.25, B=1)
      )
     )
    )
    Particles.SecondaryWidth:NiagaraFloat=&Multiply_Float(A=$Particles.RibbonWidth, B=1.2)
   ) {isSetParameters=true}
   /SetVariables(Particles.SoftRibbonWidth:NiagaraFloat=&Multiply_Float(A=$Particles.RibbonWidth, B=2)) {isSetParameters=true}
  ]
  rendererClasses [&NiagaraRibbonRendererProperties]
  renderers [{rendererClass=&NiagaraRibbonRendererProperties}]
 }
}
summary {
 systemName=NS_Pickaxe_ProdigySage_Trail
 emitters [
  {emitterName=Blue_Underaly, rendererClasses [&NiagaraRibbonRendererProperties]}
  {emitterName=Pink_Underlay, rendererClasses [&NiagaraRibbonRendererProperties]}
  {emitterName=Spikes_Trail001, rendererClasses [&NiagaraRibbonRendererProperties]}
  {emitterName=Center_Trail, rendererClasses [&NiagaraRibbonRendererProperties]}
 ]
 userVariables [
  User.bIsFrontEnd {type=NiagaraBool}
  User.bIsFrontEndPreview {type=NiagaraBool}
  User.Gems_EffectAmount {type=NiagaraFloat, defaultValue=0, description=""}
  User.Gems_EffectIntensity {type=NiagaraFloat, defaultValue=1, description=""}
  User.RibbonIDOffset {type=NiagaraInt32}
  User.SkeletalMesh {
   type=NiagaraDataInterfaceSkeletalMesh
   defaultValue {
    struct=NiagaraExt_VariableValue_DataInterface
    value {dataInterface=&self:NiagaraDataInterfaceSkeletalMesh_0}
   }
  }
 ]
}
