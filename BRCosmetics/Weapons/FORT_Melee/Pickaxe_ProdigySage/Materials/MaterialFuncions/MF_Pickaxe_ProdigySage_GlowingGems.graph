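# Material function for the ProdigySage pickaxe that adjusts base color and roughness on its crystals and drives their staged glow.
# GlowingGems_EffectAmount, overridden by ReactivityStatus, lights gem regions across five successive ramps, split by mask textures and UV square masks.
# Emissive combines bump-offset scrolling cloud noise with a sine pulse, both scaled by time of day and tinted by the base color.
# The last stage also tints base color.
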MaterialFunction /BRCosmetics/Weapons/FORT_Melee/Pickaxe_ProdigySage/Materials/MaterialFuncions/MF_Pickaxe_ProdigySage_GlowingGems
graph {nodes [
  n1 = FunctionInput() {inputName=BaseColor}
  n2 = Multiply(A=Constant3Vector() {constant=<rgba 0.0052 0.1714 0.3968 1>}) {constB=5}
  n3 = Constant() {r=-2}
  n4 = If(
   A=ScalarParameter() {group=GlowingGems, parameterName=ReactivityStatus, sliderMax=5, sortPriority=20}
   "A > B"=Constant()
   "A == B"=Constant() {r=5}
   "A < B"=ScalarParameter() {group=GlowingGems, parameterName=GlowingGems_EffectAmount, sliderMax=5}
  ) {constB=1}
  n5 = TextureSample() {samplerType=SAMPLERTYPE_Masks, texture=/BRCosmetics/Weapons/FORT_Melee/Pickaxe_ProdigySage/Textures/T_Pickaxe_ProdigySage_AdditionalMasks}
  n6 = TextureSample() {samplerType=SAMPLERTYPE_Masks, texture=/BRCosmetics/Weapons/FORT_Melee/Pickaxe_ProdigySage/Textures/T_Pickaxe_ProdigySage_GemsMask}
  n7 = Multiply(A=Saturate(None=InverseLinearInterpolate(Value=@n4) {constA=4, constB=5}), B=Multiply(A=@n5.B, B=@n6.R))
  n8 = LinearInterpolate(A=@n1, B=Multiply(A=@n1, B=Desaturation(None=@n2, Fraction=@n3)) {constB=2}, Alpha=@n7)
  n9 = Desaturation(None=@n8, Fraction=ScalarParameter() {group=GlowingGems, parameterName=CrystalsDesaturation, sliderMax=1, sortPriority=33})
  n10 = ScalarParameter() {defaultValue=1, group=GlowingGems, parameterName=CrystalsBaseColorMultiplier, sliderMax=1, sortPriority=34}
  n11 = TextureSample() {samplerType=SAMPLERTYPE_Masks, texture=&T_Pickaxe_ProdigySage_AdditionalMasks}
  n12 = TextureSample() {samplerType=SAMPLERTYPE_Masks, texture=&T_Pickaxe_ProdigySage_GemsMask}
  n13 = Multiply(A=Add(A=@n11.R, B=@n11.R), B=Saturate(None=Power(Base=@n12.R) {constExponent=0}))
  FunctionOutput(None=LinearInterpolate(A=@n8, B=Multiply(A=@n9, B=@n10), Alpha=@n13)) {outputName=BaseColor}
  n14 = Desaturation(None=@n8, Fraction=Constant() {r=-1})
  n15 = Multiply(A=Time(), B=Multiply(A=ScalarParameter() {defaultValue=1, group=GlowingGems, parameterName=NoiseSpeed}) {constB=0.02}) {constB=0.02}
  n16 = Constant2Vector() {g=1}
  n17 = TextureSample(UVs=Add(A=TextureCoordinate(), B=Multiply(A=@n15, B=@n16) {constB=0.1})) {texture=&AtmosphericCloudNoise01}
  n18 = BumpOffset(Coordinate=Add(A=TextureCoordinate(), B=Multiply(A=@n17.R) {constB=0.2}) {constB=0}, Height=Constant() {r=2.5})
  n19 = Multiply(
   A=SmoothStep(
    Min=ScalarParameter() {group=GlowingGems, parameterName=NoiseContrast}
    Value=TextureSample(UVs=@n18) {texture=&AtmosphericCloudNoise01}.RGB
   )
   B=ScalarParameter() {group=GlowingGems, parameterName=NoiseOpacity}
  ) {constB=0.7}
  n20 = Constant() {r=1}
  n21 = MaterialFunctionCall(Value_Morning=@n20, Value_Day=@n20, Value_Evening=@n20, Value_Night=Constant() {r=0.4}) {materialFunction=&Scalar_Time_of_Day}
  n22 = Multiply(A=Sine(None=Time()) {period=2.3}) {constB=0.5}
  n23 = Constant() {desc=TimeOfDay, r=5}
  n24 = Constant() {r=4}
  n25 = Constant() {r=1.75}
  n26 = If(
   A=ScalarParameter() {group=GlowingGems, parameterName=ReactivityStatus, sliderMax=5, sortPriority=20}
   "A > B"=Constant()
   "A == B"=Constant() {r=1}
   "A < B"=ScalarParameter() {defaultValue=1, group=GlowingGems, parameterName=GlowingGems_Intensity, sliderMax=2}
  ) {constB=1}
  n27 = Multiply(
   A=LinearInterpolate(Alpha=Add(A=@n22) {constB=0.5}) {constA=0.6}
   B=Multiply(
    A=MaterialFunctionCall(Value_Morning=@n23, Value_Day=@n23, Value_Evening=@n24, Value_Night=@n25) {materialFunction=&Scalar_Time_of_Day}.Result
    B=Multiply(A=@n26) {constB=2}
   )
  )
  n28 = ScalarParameter() {defaultValue=1, group=GlowingGems, parameterName=OverallGlowMultiplier, sortPriority=31}
  n29 = Multiply(A=Add(A=Multiply(A=@n19, B=@n21.Result), B=Multiply(A=@n27, B=@n28)))
  n30 = MaterialFunctionCall!(
   Texture Coordinate=TextureCoordinate()
   X Pos Right=Constant() {desc=Square X Right Position, r=0.78}
   X Pos Left=Constant() {desc=Square X Left Position, r=0.34}
   Y Pos Up=Constant() {desc=Square Y Up Position, r=0.21}
   Y Pos Down=Constant() {desc=Square Y Down Position, r=0.44}
  ) {materialFunction=&/MF_Pickaxe_ProdigySage_SquareMaskGenerator}
  n31 = OneMinus(None=@n30.SquareMask)
  n32 = Multiply(A=Saturate(None=InverseLinearInterpolate(Value=@n4) {constA=3, constB=4}), B=@n11.R)
  n33 = Saturate(None=InverseLinearInterpolate(Value=@n4))
  n34 = MaterialFunctionCall(
   Texture Coordinate=TextureCoordinate()
   X Pos Right=Constant() {desc=Square X Right Position, r=0.85}
   X Pos Left=Constant() {desc=Square X Left Position, r=0.42}
   Y Pos Up=Constant() {desc=Square Y Up Position, r=0.09}
   Y Pos Down=Constant() {desc=Square Y Down Position, r=0.3}
  ) {materialFunction=&MF_Pickaxe_ProdigySage_SquareMaskGenerator}
  n35 = Saturate(None=InverseLinearInterpolate(Value=@n4) {constA=1, constB=2})
  n36 = Add(A=Multiply(A=Multiply(A=@n33, B=@n11.G), B=@n34.SquareMask), B=Multiply(A=Multiply(A=@n35, B=@n11.R), B=@n30.SquareMask))
  n37 = InverseLinearInterpolate(Value=@n4) {constA=2, constB=3}
  n38 = Multiply(A=Multiply(A=Saturate(None=@n37), B=@n11.G), B=OneMinus(None=@n34.SquareMask))
  n39 = Add(A=Add(A=Multiply(A=@n31, B=@n32), B=Add(A=@n36, B=@n38)), B=@n7)
  FunctionOutput(None=Multiply(A=Multiply(A=@n14, B=@n29) {constB=5, desc=GlowAmount}, B=Multiply(A=@n39, B=@n6.R)) {constB=2}) {outputName=EmissiveColor, sortPriority=2}
  n40 = FunctionInput() {inputName=Roughness, inputType=FunctionInput_Scalar, sortPriority=1}
  n41 = Add(A=@n40, B=ScalarParameter() {group=GlowingGems, parameterName=CrystalsBaseRoughnessAdd, sliderMax=1, sortPriority=34}) {constB=0.25}
  FunctionOutput(None=LinearInterpolate(A=@n40, B=Saturate(None=@n41), Alpha=@n13)) {outputName=Roughness, sortPriority=1}
 ]}
