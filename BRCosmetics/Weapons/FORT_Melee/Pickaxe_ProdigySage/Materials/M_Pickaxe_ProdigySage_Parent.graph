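# Parent material for the ProdigySage pickaxe, layering a glowing-gem effect on the shared 2023 character material.
# The CharacterParent PreFX and PostFX functions build the base material attributes, which feed the pickaxe's GlowingGems function.
# Its emissive passes through cosmetic eye adaptation, then base color, roughness and emissive are written back into the attributes.
# Shading model comes from the graph, with skeletal mesh, mesh deformer and Nanite usage enabled.
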Material /BRCosmetics/Weapons/FORT_Melee/Pickaxe_ProdigySage/Materials/M_Pickaxe_ProdigySage_Parent
graph {
 properties {
  bUseMaterialAttributes=true
  bUsedWithMeshDeformer=true
  bUsedWithNanite=true
  bUsedWithSkeletalMesh=true
  pixelDepthOffsetMode=PDOM_Legacy
  shadingModel=MSM_FromMaterialExpression
  usedShadingModels=Default Lit | Subsurface | Clear Coat | Subsurface Profile | Cloth
 }
 nodes [
  n1 = MaterialFunctionCall() {materialFunction=&MF_CharacterParent_2023_PreFX}
  n2 = GetMaterialAttributes(None=MaterialFunctionCall(Material=@n1.Result, M=@n1.Result, Pre Skinned Local Position=@n1.Result, Distance Blend=@n1.Result, Fresnel=@n1.Result) {materialFunction=&MF_CharacterParent_2023_PostFX}.Result) {attributeGetTypes [#1, #2]}
  n3 = MaterialFunctionCall!(BaseColor=@n2.Base Color, Roughness=@n2.Base Color) {materialFunction=/BRCosmetics/Weapons/FORT_Melee/Pickaxe_ProdigySage/Materials/MaterialFuncions/MF_Pickaxe_ProdigySage_GlowingGems}
  n4 = MaterialFunctionCall!(Emissive=@n3.EmissiveColor, Alpha=Constant() {r=1}, Material) {materialFunction=/Game/Packages/Fortress_SharedMaterials/Base_Material_Functions/BaseCharacterShaderFunctions/MF_CosmeticEmissiveEyeAdaptation}
  SetMaterialAttributes!(MaterialAttributes=@n2.MaterialAttributes, Base Color=@n3.BaseColor, Roughness=@n3.BaseColor, Emissive Color=@n4.Emissive) {attributeSetTypes [#1, #2, #3]}
 ]
}
