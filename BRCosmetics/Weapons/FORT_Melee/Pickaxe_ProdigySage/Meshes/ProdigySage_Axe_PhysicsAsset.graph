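# Physics asset for the ProdigySage pickaxe axe mesh, a two-handed melee weapon cosmetic, giving it a simplified collision shape.
# It holds a single body on the Melee_TwoHanded_Handle bone, built from one large rotated capsule offset along the handle to roughly enclose the weapon.
# There are no constraints, so it defines no jointed or ragdoll-style motion between parts.
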PhysicsAsset /BRCosmetics/Weapons/FORT_Melee/Pickaxe_ProdigySage/Meshes/ProdigySage_Axe_PhysicsAsset
bodies {Melee_TwoHanded_Handle [{shapeType=Capsule, length=81.8683, radius=40.6986, center=<0.7982 -1.0037 38.586>, rotation=<rot 0.0022 -90 -0.0013>}]}
constraints []
